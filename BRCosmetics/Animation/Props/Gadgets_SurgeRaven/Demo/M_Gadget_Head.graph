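# M_Gadget_Head is a Default Lit material, by name for the head of the Surge Raven gadget prop in a cosmetics animation demo.
# Its graph holds a single constant purple color vector, with no textures, parameters or shown connection to a material output.
# Automatic usage-flag setting in the editor is turned off.
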Material /BRCosmetics/Animation/Props/Gadgets_SurgeRaven/Demo/M_Gadget_Head
graph {
 properties {bAutomaticallySetUsageInEditor=false, usedShadingModels=Default Lit}
 nodes [Constant3Vector() {constant=<rgba 0.3813 0.1195 0.7084 1>}]
}
