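# M_Gadget_White is a Default Lit material in the SurgeRaven gadget demo prop folder, named as a white gadget surface.
# Its graph holds a single constant color node in near-white grey, with no textures or parameters and no connection shown to a material output.
# Automatic setting of usage flags in the editor is turned off.
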Material /BRCosmetics/Animation/Props/Gadgets_SurgeRaven/Demo/M_Gadget_White
graph {
 properties {bAutomaticallySetUsageInEditor=false, usedShadingModels=Default Lit}
 nodes [Constant3Vector() {constant=<rgba 0.8714 0.8714 0.8714 1>}]
}
