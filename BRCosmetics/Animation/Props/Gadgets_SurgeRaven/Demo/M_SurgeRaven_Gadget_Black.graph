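# Flat black Default Lit material for the SurgeRaven gadget prop, kept in the cosmetics animation Demo folder.
# The graph holds a single near-black constant color node, with no textures, parameters or visible link to a material output.
# Automatic setting of usage flags in the editor is turned off.
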Material /BRCosmetics/Animation/Props/Gadgets_SurgeRaven/Demo/M_SurgeRaven_Gadget_Black
graph {
 properties {bAutomaticallySetUsageInEditor=false, usedShadingModels=Default Lit}
 nodes [Constant3Vector() {constant=<rgba 0.0208 0.0208 0.0208 0>}]
}
